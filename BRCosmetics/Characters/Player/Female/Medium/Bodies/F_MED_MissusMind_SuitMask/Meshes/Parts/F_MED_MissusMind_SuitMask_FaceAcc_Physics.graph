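# Physics asset for the face accessory of the female Missus Mind SuitMask body, modelling left and right mask tubes as capsule chains.
# Each chain pivots freely from the head, has angular limits on its upper segments, and locks its lower segments through to a shared end anchor.
# Extra constraints tie segments 04 and 05 to the same-side clavicle.
# Capsules also cover the head, neck, upper spine, clavicles and upper arms.
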PhysicsAsset /BRCosmetics/Characters/Player/Female/Medium/Bodies/F_MED_MissusMind_SuitMask/Meshes/Parts/F_MED_MissusMind_SuitMask_FaceAcc_Physics
bodies {
 clavicle_l [{
   shapeType=Capsule
   length=34.4228
   radius=7.5164
   shapeName=clavicle_l_capsule
   center=<12.7894 0 -3.1812>
   rotation=<rot 80.5075 -0.0045 -0.0047>
  }]
 clavicle_r [{
   shapeType=Capsule
   length=34.0244
   radius=7.9148
   shapeName=clavicle_r_capsule
   center=<-12.7345 0 3.457>
   rotation=<rot -83.3584 179.9928 -179.9927>
  }]
 dyn_l_tube_01 [{
   shapeType=Capsule
   length=5.2024
   radius=1.1606
   shapeName=dyn_l_tube_01_capsule
   center=<2.9813 0.09 0.2827>
   rotation=<rot 0 -88.6624 -89.7268>
  }]
 dyn_l_tube_02 [{
   shapeType=Capsule
   length=5.44
   radius=1.1606
   shapeName=dyn_l_tube_02_capsule
   center=<3.5564 0.1317 0.15>
   rotation=<rot 0.2494 -95.3858 -86.5794>
  }]
 dyn_l_tube_03 [{
   shapeType=Capsule
   length=5.5078
   radius=1.1606
   shapeName=dyn_l_tube_02_capsule
   center=<3.1611 -0.1748 0.2082>
   rotation=<rot -0.0543 -90.7827 -82.8555>
  }]
 dyn_l_tube_04 [{shapeType=Capsule, length=4.305, radius=1.1606, shapeName=dyn_l_tube_04_capsule, center=<2.335 -0.4027 0.2161>, rotation=<rot -90 180 180>}]
 dyn_l_tube_05 [{
   shapeType=Capsule
   length=5.0628
   radius=1.1606
   shapeName=dyn_l_tube_05_capsule
   center=<2.2512 -0.1244 0.2406>
   rotation=<rot -0.0001 96.4852 -87.2591>
  }]
 dyn_l_tube_06 [{shapeType=Capsule, length=3.7866, radius=1.1606, shapeName=dyn_l_tube_06_capsule, center=<1.2315 0 0>, rotation=<rot -90 180 180>}]
 dyn_l_tube_mstr [{shapeType=Capsule, length=0.505, radius=0.505, shapeName=dyn_l_tube_mstr_capsule, rotation=<rot 0 -90 0>}]
 dyn_r_tube_01 [{
   shapeType=Capsule
   length=5.0771
   radius=1.16
   shapeName=dyn_l_tube_01_capsule
   center=<2.979 0.09 -0.1903>
   rotation=<rot 0 -88.6624 -89.7268>
  }]
 dyn_r_tube_02 [{
   shapeType=Capsule
   length=5.3147
   radius=1.16
   shapeName=dyn_l_tube_02_capsule
   center=<3.5673 0.1296 -0.0983>
   rotation=<rot 0.2495 -95.3857 -92.5445>
  }]
 dyn_r_tube_03 [{
   shapeType=Capsule
   length=4.8911
   radius=1.16
   shapeName=dyn_l_tube_02_capsule
   center=<3.3442 -0.177 -0.0757>
   rotation=<rot -0.0543 -90.7825 -90.998>
  }]
 dyn_r_tube_04 [{shapeType=Capsule, length=3.9994, radius=1.2186, shapeName=dyn_l_tube_04_capsule, center=<2.335 -0.2566 -0.0622>, rotation=<rot -90 0 0>}]
 dyn_r_tube_05 [{
   shapeType=Capsule
   length=4.8788
   radius=1.2186
   shapeName=dyn_l_tube_05_capsule
   center=<2.2562 0.0869 -0.3618>
   rotation=<rot 0 96.4852 -93.4456>
  }]
 dyn_r_tube_06 [{shapeType=Capsule, length=3.6613, radius=1.16, shapeName=dyn_l_tube_06_capsule, center=<1.2315 0 0>, rotation=<rot -90 180 180>}]
 dyn_r_tube_mstr [{shapeType=Capsule, length=0.505, radius=0.505, shapeName=dyn_r_tube_mstr_capsule, rotation=<rot 0 -90 0>}]
 head [{shapeType=Capsule, length=7.3486, radius=6.6514, shapeName=head_capsule, center=<4.44 0 0>, rotation=<rot 90 0 0>}]
 neck_01 [{shapeType=Capsule, length=9.0432, radius=4.9568, shapeName=neck_01_capsule, center=<1.3039 0.3361 0>, rotation=<rot 90 0 0>}]
 neck_02 [{shapeType=Capsule, length=9.0432, radius=4.9568, shapeName=neck_02_capsule, center=<1.3039 0.3361 0>, rotation=<rot 90 0 0>}]
 spine_05 [{shapeType=Capsule, length=9.9313, radius=11.1302, shapeName=spine_05_capsule, center=<0.1773 2.1853 0>}]
 tube_end_anchor [{shapeType=Capsule, length=0.505, radius=0.505, shapeName=tube_end_anchor_capsule, rotation=<rot 0 -90 0>}]
 upperarm_l [{
   shapeType=Capsule
   length=33.6017
   radius=4.2364
   shapeName=upperarm_l_capsule
   center=<17.7532 0 0.5631>
   rotation=<rot 84.758 -0.4815 -0.5266>
  }]
 upperarm_r [{
   shapeType=Capsule
   length=33.6017
   radius=4.2364
   shapeName=upperarm_r_capsule
   center=<-16.8578 -0.0001 0.2267>
   rotation=<rot 83.5669 -0.1444 -0.1616>
  }]
}
constraints [
 {
  bone1Name=dyn_l_tube_mstr
  bone2Name=head
  swing1LimitDegrees=45
  swing1Motion=Free
  swing2LimitDegrees=45
  swing2Motion=Free
  twistLimitDegrees=45
  twistMotion=Free
 }
 {bone1Name=dyn_l_tube_01, bone2Name=dyn_l_tube_mstr, swing1LimitDegrees=30, swing2LimitDegrees=30, twistLimitDegrees=2}
 {bone1Name=dyn_l_tube_02, bone2Name=dyn_l_tube_01, swing1LimitDegrees=30, swing2LimitDegrees=30, twistLimitDegrees=2}
 {bone1Name=dyn_l_tube_03, bone2Name=dyn_l_tube_02, swing1LimitDegrees=30, swing2LimitDegrees=30, twistLimitDegrees=2}
 {
  bone1Name=dyn_l_tube_04
  bone2Name=dyn_l_tube_03
  swing1LimitDegrees=10
  swing1Motion=Locked
  swing2LimitDegrees=10
  swing2Motion=Locked
  twistLimitDegrees=2
  twistMotion=Locked
 }
 {
  bone1Name=dyn_l_tube_05
  bone2Name=dyn_l_tube_04
  swing1LimitDegrees=10
  swing1Motion=Locked
  swing2LimitDegrees=10
  swing2Motion=Locked
  twistLimitDegrees=2
  twistMotion=Locked
 }
 {
  bone1Name=dyn_l_tube_06
  bone2Name=dyn_l_tube_05
  swing1LimitDegrees=10
  swing1Motion=Locked
  swing2LimitDegrees=10
  swing2Motion=Locked
  twistLimitDegrees=2
  twistMotion=Locked
 }
 {
  bone1Name=dyn_r_tube_mstr
  bone2Name=head
  swing1LimitDegrees=45
  swing1Motion=Free
  swing2LimitDegrees=45
  swing2Motion=Free
  twistLimitDegrees=45
  twistMotion=Free
 }
 {bone1Name=dyn_r_tube_01, bone2Name=dyn_r_tube_mstr, swing1LimitDegrees=30, swing2LimitDegrees=30, twistLimitDegrees=2}
 {bone1Name=dyn_r_tube_02, bone2Name=dyn_r_tube_01, swing1LimitDegrees=30, swing2LimitDegrees=30, twistLimitDegrees=2}
 {bone1Name=dyn_r_tube_03, bone2Name=dyn_r_tube_02, swing1LimitDegrees=30, swing2LimitDegrees=30, twistLimitDegrees=2}
 {
  bone1Name=dyn_r_tube_04
  bone2Name=dyn_r_tube_03
  swing1LimitDegrees=10
  swing1Motion=Locked
  swing2LimitDegrees=10
  swing2Motion=Locked
  twistLimitDegrees=2
  twistMotion=Locked
 }
 {
  bone1Name=dyn_r_tube_05
  bone2Name=dyn_r_tube_04
  swing1LimitDegrees=10
  swing1Motion=Locked
  swing2LimitDegrees=10
  swing2Motion=Locked
  twistLimitDegrees=2
  twistMotion=Locked
 }
 {
  bone1Name=dyn_r_tube_06
  bone2Name=dyn_r_tube_05
  swing1LimitDegrees=10
  swing1Motion=Locked
  swing2LimitDegrees=10
  swing2Motion=Locked
  twistLimitDegrees=2
  twistMotion=Locked
 }
 {
  bone1Name=dyn_l_tube_06
  bone2Name=tube_end_anchor
  swing1LimitDegrees=45
  swing1Motion=Locked
  swing2LimitDegrees=45
  swing2Motion=Locked
  twistLimitDegrees=45
  twistMotion=Locked
 }
 {
  bone1Name=dyn_r_tube_06
  bone2Name=tube_end_anchor
  swing1LimitDegrees=45
  swing1Motion=Locked
  swing2LimitDegrees=45
  swing2Motion=Locked
  twistLimitDegrees=45
  twistMotion=Locked
 }
 {
  bone1Name=upperarm_l
  bone2Name=clavicle_l
  swing1LimitDegrees=45
  swing1Motion=Free
  swing2LimitDegrees=45
  swing2Motion=Free
  twistLimitDegrees=45
  twistMotion=Free
 }
 {
  bone1Name=upperarm_r
  bone2Name=clavicle_r
  swing1LimitDegrees=45
  swing1Motion=Free
  swing2LimitDegrees=45
  swing2Motion=Free
  twistLimitDegrees=45
  twistMotion=Free
 }
 {
  bone1Name=dyn_l_tube_05
  bone2Name=clavicle_l
  swing1LimitDegrees=45
  swing1Motion=Locked
  swing2LimitDegrees=45
  swing2Motion=Locked
  twistLimitDegrees=45
  twistMotion=Locked
 }
 {
  bone1Name=dyn_r_tube_05
  bone2Name=clavicle_r
  swing1LimitDegrees=45
  swing1Motion=Locked
  swing2LimitDegrees=45
  swing2Motion=Locked
  twistLimitDegrees=45
  twistMotion=Locked
 }
 {bone1Name=dyn_l_tube_04, bone2Name=clavicle_l, swing1LimitDegrees=45, swing2LimitDegrees=45, twistLimitDegrees=45}
 {bone1Name=dyn_r_tube_04, bone2Name=clavicle_r, swing1LimitDegrees=45, swing2LimitDegrees=45, twistLimitDegrees=45}
]
