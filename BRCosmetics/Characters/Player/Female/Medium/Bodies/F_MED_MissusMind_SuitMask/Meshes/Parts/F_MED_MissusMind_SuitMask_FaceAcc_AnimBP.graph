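# Anim Blueprint for the Missus Mind SuitMask female face accessory, adding physics secondary motion to the player cosmetic.
# It copies the attached parent's pose into a RigidBody node simulating relative to the head bone.
# Each update, a deprecated gravity state function sets the node's gravity override and stores per-state RBAN settings, which the graph never reads.
# Zipline events and skydiving set RBAN Alpha, whose getter appears unconnected to the RigidBody node.
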AnimBlueprint /BRCosmetics/Characters/Player/Female/Medium/Bodies/F_MED_MissusMind_SuitMask/Meshes/Parts/F_MED_MissusMind_SuitMask_FaceAcc_AnimBP
target_skeleton=&/F_MED_MissusMind_SuitMask_FaceAcc_Skeleton
graphs [
 AnimGraph {nodes [
   n1 = RigidBody!(
    ComponentPose=Local To Component(LocalPose=Copy Pose From Mesh() {bUseAttachedParent=true})
    Alpha
    OverrideWorldGravity:Vector=Get Gravity Override()
    ExternalForce
    ComponentLinearAccScale:Vector="0.4,0.4,0.4"
    ComponentLinearVelScale:Vector="0.5,0.5,0.5"
    SimSpaceSettings:Sim Space Settings Structure=(WorldAlpha=0,VelocityScaleZ=1,DampingAlpha=1,MaxLinearVelocity=300,MaxAngularVelocity=10,MaxLinearAcceleration=100,MaxAngularAcceleration=5,ExternalLinearDragV=(X=0,Y=0,Z=0),ExternalLinearVelocity=(X=0,Y=0,Z=0),ExternalAngularVelocity=(X=0,Y=0,Z=0))
   ) -> (Pose) {
    bClampLinearTranslationLimitToRefPose=true
    bOverrideWorldGravity=true
    overridePhysicsAsset=&/F_MED_MissusMind_SuitMask_FaceAcc_Physics
    simulationSpace=BaseBoneSpace
    alphaBoolBlend {blendOption=Linear}
    alphaScaleBiasClamp {interpSpeedDecreasing=50, interpSpeedIncreasing=50}
    baseBoneRef {boneName=head}
    componentAppliedLinearAccClamp=<3000 3000 3000>
    componentLinearAccScale=<0.4 0.4 0.4>
    componentLinearVelScale=<0.5 0.5 0.5>
    simSpaceSettings {maxAngularAcceleration=5, maxAngularVelocity=10, maxLinearAcceleration=100, maxLinearVelocity=300}
   }
   Output Pose(Result=Component To Local(ComponentPose=@n1)) {layerGroup=None}
   Get RBAN Alpha() -> (RBAN Alpha:double)
  ]}
 EventGraph {nodes [
   n1 = Event BlueprintUpdateAnimation()
   > n2 = Sequence()
   > n3 = Gravity Override State Function DEPRECATED(
    CharacterPartAnimInstance=Self-Reference()
    GameTimeInSeconds=@n1.DeltaTimeX
    GravityOverrideParams=Get Gravity Override Params()
    RBAN Params=Get RBAN Params()
   )
   > Set Gravity Override(Gravity Override=@n3)
   > Set RBAN Settings Out(RBAN Settings Out:RBANSetting Struct Structure=@n3.RBAN SettingsOut) -> (Output_Get:RBANSetting Struct Structure)
   n4 = Branch(execute=@n2.then_1, Condition=OR Boolean(A=Get isOnZipline() -> (isOnZipline:bool), B=Get bIsSkyDiving()))
   > Set RBAN Alpha(RBAN Alpha:double=0.9) -> (Output_Get:double)
   Set RBAN Alpha(execute=@n4.else, RBAN Alpha=1)
   n5 = Event BlueprintProcessAnimInputEvent()
   > n6 = Branch(Condition=IsNamed(self=@n5.AnimInputEvent, InEventName=aie.Player.Zipline.Entered))
   > Set isOnZipline(isOnZipline:bool=1) -> (Output_Get:bool)
   Branch(execute=@n6.else, Condition=IsNamed(self=@n5.AnimInputEvent, InEventName=aie.Player.Zipline.Exited))
   > Set isOnZipline(isOnZipline=false)
  ]}
]
variables [
 Gravity Override {type=Vector}
 Gravity Override Params {type=GravityOverrideParamsStruct}
 isOnZipline {type=boolean}
 RBAN Alpha {type=number}
 RBAN Params {
  type=RBANParamsStruct
  default {
   dBNO {
    componentLinearAccScale=<0.4 0.4 0.4>
    componentLinearVelScale=<0.5 0.5 0.5>
    simSpaceSettings {
     maxAngularAcceleration=5
     maxAngularVelocity=10
     maxLinearAcceleration=100
     maxLinearVelocity=300
     velocityScaleZ=1
     worldAlpha=0
     externalLinearDragV=<0 0 0>
    }
   }
   dBNO_Carried {
    componentLinearAccScale=<0.4 0.4 0.4>
    componentLinearVelScale=<0.5 0.5 0.5>
    simSpaceSettings {
     maxAngularAcceleration=5
     maxAngularVelocity=10
     maxLinearAcceleration=100
     maxLinearVelocity=300
     velocityScaleZ=1
     worldAlpha=0
     externalLinearDragV=<0 0 0>
    }
   }
   emoteOrMelee {
    componentLinearAccScale=<0.4 0.4 0.4>
    componentLinearVelScale=<0.5 0.5 0.5>
    simSpaceSettings {
     maxAngularAcceleration=5
     maxAngularVelocity=10
     maxLinearAcceleration=100
     maxLinearVelocity=300
     velocityScaleZ=1
     worldAlpha=0
     externalLinearDragV=<0 0 0>
    }
   }
   front_End {
    componentLinearAccScale=<0.4 0.4 0.4>
    componentLinearVelScale=<0.5 0.5 0.5>
    simSpaceSettings {
     maxAngularAcceleration=5
     maxAngularVelocity=10
     maxLinearAcceleration=100
     maxLinearVelocity=300
     velocityScaleZ=1
     worldAlpha=0
     externalLinearDragV=<0 0 0>
    }
   }
   onGround_Crouch_Moving {
    componentLinearAccScale=<0.4 0.4 0.4>
    componentLinearVelScale=<0.5 0.5 0.5>
    simSpaceSettings {
     maxAngularAcceleration=5
     maxAngularVelocity=10
     maxLinearAcceleration=100
     maxLinearVelocity=300
     velocityScaleZ=1
     worldAlpha=0
     externalLinearDragV=<0 0 0>
    }
   }
   onGround_Falling {
    componentLinearAccScale=<0.4 0.4 0.4>
    componentLinearVelScale=<0.5 0.5 0.5>
    simSpaceSettings {
     maxAngularAcceleration=5
     maxAngularVelocity=10
     maxLinearAcceleration=100
     maxLinearVelocity=300
     velocityScaleZ=1
     worldAlpha=0
     externalLinearDragV=<0 0 0>
    }
   }
   onGround_Moving {simSpaceSettings {
     maxAngularAcceleration=5
     maxAngularVelocity=10
     maxLinearAcceleration=100
     maxLinearVelocity=300
     velocityScaleZ=1
     worldAlpha=0
     externalLinearDragV=<0 0 0>
    }}
   onGround_Standing {
    componentLinearAccScale=<0.4 0.4 0.4>
    componentLinearVelScale=<0.5 0.5 0.5>
    simSpaceSettings {
     maxAngularAcceleration=5
     maxAngularVelocity=10
     maxLinearAcceleration=100
     maxLinearVelocity=300
     velocityScaleZ=1
     worldAlpha=0
     externalLinearDragV=<0 0 0>
    }
   }
   rail_Grinding {
    componentLinearAccScale=<0.4 0.4 0.4>
    componentLinearVelScale=<0.5 0.5 0.5>
    simSpaceSettings {
     maxAngularAcceleration=5
     maxAngularVelocity=10
     maxLinearAcceleration=100
     maxLinearVelocity=300
     velocityScaleZ=1
     worldAlpha=0
     externalLinearDragV=<0 0 0>
    }
   }
   riding_Moving {
    componentLinearAccScale=<0.4 0.4 0.4>
    componentLinearVelScale=<0.5 0.5 0.5>
    simSpaceSettings {
     maxAngularAcceleration=5
     maxAngularVelocity=10
     maxLinearAcceleration=100
     maxLinearVelocity=300
     velocityScaleZ=1
     worldAlpha=0
     externalLinearDragV=<0 0 0>
    }
   }
   riding_Moving_Targeting {
    componentLinearAccScale=<0.4 0.4 0.4>
    componentLinearVelScale=<0.5 0.5 0.5>
    simSpaceSettings {
     maxAngularAcceleration=5
     maxAngularVelocity=10
     maxLinearAcceleration=100
     maxLinearVelocity=300
     velocityScaleZ=1
     worldAlpha=0
     externalLinearDragV=<0 0 0>
    }
   }
   riding_Standing {
    componentLinearAccScale=<0.4 0.4 0.4>
    componentLinearVelScale=<0.5 0.5 0.5>
    simSpaceSettings {
     maxAngularAcceleration=5
     maxAngularVelocity=10
     maxLinearAcceleration=100
     maxLinearVelocity=300
     velocityScaleZ=1
     worldAlpha=0
     externalLinearDragV=<0 0 0>
    }
   }
   skydive_DivingDown {
    componentLinearAccScale=<0.4 0.4 0.4>
    componentLinearVelScale=<0.5 0.5 0.5>
    simSpaceSettings {
     maxAngularAcceleration=5
     maxAngularVelocity=10
     maxLinearAcceleration=100
     maxLinearVelocity=300
     velocityScaleZ=1
     worldAlpha=0
     externalLinearDragV=<0 0 0>
    }
   }
   skydive_FloatingDown {
    componentLinearAccScale=<0.4 0.4 0.4>
    componentLinearVelScale=<0.5 0.5 0.5>
    simSpaceSettings {
     maxAngularAcceleration=5
     maxAngularVelocity=10
     maxLinearAcceleration=100
     maxLinearVelocity=300
     velocityScaleZ=1
     worldAlpha=0
     externalLinearDragV=<0 0 0>
    }
   }
   skydive_GoingUp {
    componentLinearAccScale=<0.4 0.4 0.4>
    componentLinearVelScale=<0.5 0.5 0.5>
    simSpaceSettings {
     maxAngularAcceleration=5
     maxAngularVelocity=10
     maxLinearAcceleration=100
     maxLinearVelocity=300
     velocityScaleZ=1
     worldAlpha=0
     externalLinearDragV=<0 0 0>
    }
   }
   skydive_Parachuting {
    componentLinearAccScale=<0.4 0.4 0.4>
    componentLinearVelScale=<0.5 0.5 0.5>
    simSpaceSettings {
     maxAngularAcceleration=5
     maxAngularVelocity=10
     maxLinearAcceleration=100
     maxLinearVelocity=300
     velocityScaleZ=1
     worldAlpha=0
     externalLinearDragV=<0 0 0>
    }
   }
   skydive_Surfing {
    componentLinearAccScale=<0.4 0.4 0.4>
    componentLinearVelScale=<0.5 0.5 0.5>
    simSpaceSettings {
     maxAngularAcceleration=5
     maxAngularVelocity=10
     maxLinearAcceleration=100
     maxLinearVelocity=300
     velocityScaleZ=1
     worldAlpha=0
     externalLinearDragV=<0 0 0>
    }
   }
   swimming {
    componentLinearAccScale=<0.4 0.4 0.4>
    componentLinearVelScale=<0.5 0.5 0.5>
    simSpaceSettings {
     maxAngularAcceleration=5
     maxAngularVelocity=10
     maxLinearAcceleration=100
     maxLinearVelocity=300
     velocityScaleZ=1
     worldAlpha=0
     externalLinearDragV=<0 0 0>
    }
   }
   tactical_Sprint {
    componentLinearAccScale=<0.4 0.4 0.4>
    componentLinearVelScale=<0.5 0.5 0.5>
    simSpaceSettings {
     maxAngularAcceleration=5
     maxAngularVelocity=10
     maxLinearAcceleration=100
     maxLinearVelocity=300
     velocityScaleZ=1
     worldAlpha=0
     externalLinearDragV=<0 0 0>
    }
   }
   ziplining {simSpaceSettings {
     maxAngularAcceleration=0
     maxAngularVelocity=0
     maxLinearAcceleration=0
     maxLinearVelocity=0
     velocityScaleZ=1
     worldAlpha=0
     externalLinearDragV=<0 0 0>
    }}
  }
 }
 RBAN Settings Out {
  type=RBANSettingStruct
  default {
   componentLinearAccScale=<0 0 0>
   componentLinearVelScale=<0 0 0>
   simSpaceSettings {
    dampingAlpha=1
    maxAngularAcceleration=1000
    maxAngularVelocity=3
    maxLinearAcceleration=300
    maxLinearVelocity=700
    velocityScaleZ=0
    worldAlpha=1
    externalAngularVelocity=<0 0 0>
    externalLinearDragV=<0 0.5 0>
    externalLinearVelocity=<0 0 0>
   }
  }
 }
]
